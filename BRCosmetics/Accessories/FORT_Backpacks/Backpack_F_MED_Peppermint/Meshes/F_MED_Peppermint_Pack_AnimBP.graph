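# Anim Blueprint for the Peppermint backpack that procedurally drifts six icicle shard bones and adds springy physics sway.
# Each update, Random Vector Function turns game time and per-shard frequency and amplitude into six offset vectors.
# The AnimGraph offsets the ref-pose pack bone, then adds those vectors as bone-space translations to the icicles.
# Chained root-relative AnimDynamics nodes then apply partially blended, linearly limited springs to pack and icicles, with zero angular limits.
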AnimBlueprint /BRCosmetics/Accessories/FORT_Backpacks/Backpack_F_MED_Peppermint/Meshes/F_MED_Peppermint_Pack_AnimBP
target_skeleton=&/F_MED_Peppermint_Pack_Skeleton
graphs [
 AnimGraph {nodes [
   n1 = "Transform (Modify) Bone - Bone: pack"(ComponentPose=Local To Component(LocalPose=Local Space Ref Pose()), Translation="0,-5,0", Rotation="0,0,0") {translationMode=BMM_Additive, alphaBoolBlend {blendOption=Linear}, boneToModify {boneName=pack}, translation {y=-5}}
   n2 = "Transform (Modify) Bone - Bone: dyn_icicle_1"(ComponentPose=@n1, Alpha=1, Translation=Get Shard 1 Movement() -> (Shard 1 Movement:Vector), Rotation="0,0,0", Scale="1,1,1") {
    rotationMode=BMM_Additive
    rotationSpace=BCS_BoneSpace
    scaleMode=BMM_Additive
    scaleSpace=BCS_BoneSpace
    translationMode=BMM_Additive
    translationSpace=BCS_BoneSpace
    alphaBoolBlend {blendOption=Linear}
    boneToModify {boneName=dyn_icicle_1}
   }
   n3 = "Transform (Modify) Bone - Bone: dyn_icicle_2"(ComponentPose=@n2, Alpha=1, Translation=Get Shard 2 Movement() -> (Shard 2 Movement:Vector), Rotation="0,0,0", Scale="1,1,1") {
    rotationMode=BMM_Additive
    rotationSpace=BCS_BoneSpace
    scaleMode=BMM_Additive
    scaleSpace=BCS_BoneSpace
    translationMode=BMM_Additive
    translationSpace=BCS_BoneSpace
    alphaBoolBlend {blendOption=Linear}
    boneToModify {boneName=dyn_icicle_2}
   }
   n4 = "Transform (Modify) Bone - Bone: dyn_icicle_3"(ComponentPose=@n3, Alpha=1, Translation=Get Shard 3 Movement() -> (Shard 3 Movement:Vector), Rotation="0,0,0", Scale="1,1,1") {
    rotationMode=BMM_Additive
    rotationSpace=BCS_BoneSpace
    scaleMode=BMM_Additive
    scaleSpace=BCS_BoneSpace
    translationMode=BMM_Additive
    translationSpace=BCS_BoneSpace
    alphaBoolBlend {blendOption=Linear}
    boneToModify {boneName=dyn_icicle_3}
   }
   n5 = "Transform (Modify) Bone - Bone: dyn_icicle_4"(ComponentPose=@n4, Alpha=1, Translation=Get Shard 4 Movement() -> (Shard 4 Movement:Vector), Rotation="0,0,0", Scale="1,1,1") {
    rotationMode=BMM_Additive
    rotationSpace=BCS_BoneSpace
    scaleMode=BMM_Additive
    scaleSpace=BCS_BoneSpace
    translationMode=BMM_Additive
    translationSpace=BCS_BoneSpace
    alphaBoolBlend {blendOption=Linear}
    boneToModify {boneName=dyn_icicle_4}
   }
   n6 = "Transform (Modify) Bone - Bone: dyn_icicle_5"(ComponentPose=@n5, Alpha=1, Translation=Get Shard 5 Movement() -> (Shard 5 Movement:Vector), Rotation="0,0,0", Scale="1,1,1") {
    rotationMode=BMM_Additive
    rotationSpace=BCS_BoneSpace
    scaleMode=BMM_Additive
    scaleSpace=BCS_BoneSpace
    translationMode=BMM_Additive
    translationSpace=BCS_BoneSpace
    alphaBoolBlend {blendOption=Linear}
    boneToModify {boneName=dyn_icicle_5}
   }
   n7 = "Transform (Modify) Bone - Bone: dyn_icicle_6"(ComponentPose=@n6, Alpha=1, Translation=Get Shard 6 Movement() -> (Shard 6 Movement:Vector), Rotation="0,0,0", Scale="1,1,1") {
    rotationMode=BMM_Additive
    rotationSpace=BCS_BoneSpace
    scaleMode=BMM_Additive
    scaleSpace=BCS_BoneSpace
    translationMode=BMM_Additive
    translationSpace=BCS_BoneSpace
    alphaBoolBlend {blendOption=Linear}
    boneToModify {boneName=dyn_icicle_6}
   }
   n8 = Anim Dynamics - Bone: pack(ComponentPose=@n7, Alpha=0.2, ExternalForce="0,0,0") {
    angularDampingOverride=1
    angularSpringConstant=20000
    bAngularSpring=true
    bGravityOverrideInSimSpace=true
    bLinearSpring=true
    bOverrideAngularDamping=true
    bOverrideLinearDamping=true
    bUseGravityOverride=true
    bUsePlanarLimit=false
    gravityScale=0.35
    linearDampingOverride=1
    linearSpringConstant=150
    simulationSpace=RootRelative
    alphaBoolBlend {blendOption=Linear}
    boundBone {boneName=pack}
    componentAppliedLinearAccClamp=<10000 10000 10000>
    componentLinearAccScale=<0.1 0.1 0.1>
    componentLinearVelScale=<0.1 0.1 0.1>
    gravityOverride {z=-50}
    physicsBodyDefinitions [{
      sphereCollisionRadius=0
      boundBone {boneName=pack}
      boxExtents=<50 20 50>
      constraintSetup {angularLimitsMax=<0 0 0>, angularLimitsMin=<0 0 0>, angularTarget=<1 0 0>, linearAxesMax=<2.5 1 2.5>, linearAxesMin=<-2.5 -1 -2.5>}
      localJointOffset=<0 0 0>
     }]
   }
   n9 = Anim Dynamics - Bone: dyn_icicle_1(ComponentPose=@n8, Alpha=0.5, ExternalForce="0,0,0") {
    angularDampingOverride=1
    angularSpringConstant=20000
    bAngularSpring=true
    bGravityOverrideInSimSpace=true
    bLinearSpring=true
    bOverrideAngularDamping=true
    bOverrideLinearDamping=true
    bUseGravityOverride=true
    bUsePlanarLimit=false
    gravityScale=0.35
    linearDampingOverride=1
    linearSpringConstant=150
    simulationSpace=RootRelative
    alphaBoolBlend {blendOption=Linear}
    boundBone {boneName=dyn_icicle_1}
    componentAppliedLinearAccClamp=<10000 10000 10000>
    componentLinearAccScale=<0.15 0.15 0.15>
    componentLinearVelScale=<0.15 0.15 0.15>
    gravityOverride {z=-50}
    physicsBodyDefinitions [{
      sphereCollisionRadius=0
      boundBone {boneName=dyn_icicle_1}
      boxExtents=<15 25 25>
      constraintSetup {angularLimitsMax=<0 0 0>, angularLimitsMin=<0 0 0>, angularTarget=<1 0 0>, linearAxesMax=<2.5 2 2.5>, linearAxesMin=<-2.5 -2 -2.5>}
      localJointOffset=<0 0 0>
     }]
    relativeSpaceBone {boneName=pack}
   }
   n10 = Anim Dynamics - Bone: dyn_icicle_2(ComponentPose=@n9, Alpha=0.3, ExternalForce="0,0,0") {
    angularDampingOverride=1
    angularSpringConstant=20000
    bAngularSpring=true
    bGravityOverrideInSimSpace=true
    bLinearSpring=true
    bOverrideAngularDamping=true
    bOverrideLinearDamping=true
    bUseGravityOverride=true
    bUsePlanarLimit=false
    gravityScale=0.35
    linearDampingOverride=1
    linearSpringConstant=150
    simulationSpace=RootRelative
    alphaBoolBlend {blendOption=Linear}
    boundBone {boneName=dyn_icicle_2}
    componentAppliedLinearAccClamp=<10000 10000 10000>
    componentLinearAccScale=<0.1 0.1 0.1>
    componentLinearVelScale=<0.1 0.1 0.1>
    gravityOverride {z=-50}
    physicsBodyDefinitions [{
      sphereCollisionRadius=0
      boundBone {boneName=dyn_icicle_2}
      boxExtents=<13 22 22>
      constraintSetup {angularLimitsMax=<0 0 0>, angularLimitsMin=<0 0 0>, angularTarget=<1 0 0>, linearAxesMax=<2.2 2 2.2>, linearAxesMin=<-2.2 -2 -2.2>}
      localJointOffset=<0 0 0>
     }]
    relativeSpaceBone {boneName=pack}
   }
   n11 = Anim Dynamics - Bone: dyn_icicle_3(ComponentPose=@n10, Alpha=0.25, ExternalForce="0,0,0") {
    angularDampingOverride=1
    angularSpringConstant=20000
    bAngularSpring=true
    bGravityOverrideInSimSpace=true
    bLinearSpring=true
    bOverrideAngularDamping=true
    bOverrideLinearDamping=true
    bUseGravityOverride=true
    bUsePlanarLimit=false
    gravityScale=0.35
    linearDampingOverride=1
    linearSpringConstant=150
    simulationSpace=RootRelative
    alphaBoolBlend {blendOption=Linear}
    boundBone {boneName=dyn_icicle_3}
    componentAppliedLinearAccClamp=<10000 10000 10000>
    componentLinearAccScale=<0.15 0.15 0.15>
    componentLinearVelScale=<0.15 0.15 0.15>
    gravityOverride {z=-50}
    physicsBodyDefinitions [{
      sphereCollisionRadius=0
      boundBone {boneName=dyn_icicle_3}
      boxExtents=<10 18 18>
      constraintSetup {angularLimitsMax=<0 0 0>, angularLimitsMin=<0 0 0>, angularTarget=<1 0 0>, linearAxesMax=<2.5 2 2.5>, linearAxesMin=<-2.5 -2 -2.5>}
      localJointOffset=<0 0 0>
     }]
    relativeSpaceBone {boneName=pack}
   }
   n12 = Anim Dynamics - Bone: dyn_icicle_4(ComponentPose=@n11, Alpha=0.5, ExternalForce="0,0,0") {
    angularDampingOverride=1
    angularSpringConstant=20000
    bAngularSpring=true
    bGravityOverrideInSimSpace=true
    bLinearSpring=true
    bOverrideAngularDamping=true
    bOverrideLinearDamping=true
    bUseGravityOverride=true
    bUsePlanarLimit=false
    gravityScale=0.35
    linearDampingOverride=1
    linearSpringConstant=150
    simulationSpace=RootRelative
    alphaBoolBlend {blendOption=Linear}
    boundBone {boneName=dyn_icicle_4}
    componentAppliedLinearAccClamp=<10000 10000 10000>
    componentLinearAccScale=<0.1 0.1 0.1>
    componentLinearVelScale=<0.15 0.15 0.15>
    gravityOverride {z=-50}
    physicsBodyDefinitions [{
      sphereCollisionRadius=0
      boundBone {boneName=dyn_icicle_4}
      boxExtents=<15 25 25>
      constraintSetup {angularLimitsMax=<0 0 0>, angularLimitsMin=<0 0 0>, angularTarget=<1 0 0>, linearAxesMax=<2.5 2 2.5>, linearAxesMin=<-2.5 -2 -2.5>}
      localJointOffset=<0 0 0>
     }]
    relativeSpaceBone {boneName=pack}
   }
   n13 = Anim Dynamics - Bone: dyn_icicle_5(ComponentPose=@n12, Alpha=0.35, ExternalForce="0,0,0") {
    angularDampingOverride=1
    angularSpringConstant=20000
    bAngularSpring=true
    bGravityOverrideInSimSpace=true
    bLinearSpring=true
    bOverrideAngularDamping=true
    bOverrideLinearDamping=true
    bUseGravityOverride=true
    bUsePlanarLimit=false
    gravityScale=0.35
    linearDampingOverride=1
    linearSpringConstant=150
    simulationSpace=RootRelative
    alphaBoolBlend {blendOption=Linear}
    boundBone {boneName=dyn_icicle_5}
    componentAppliedLinearAccClamp=<10000 10000 10000>
    componentLinearAccScale=<0.1 0.1 0.1>
    componentLinearVelScale=<0.15 0.15 0.15>
    gravityOverride {z=-50}
    physicsBodyDefinitions [{
      sphereCollisionRadius=0
      boundBone {boneName=dyn_icicle_5}
      boxExtents=<10 18 18>
      constraintSetup {angularLimitsMax=<0 0 0>, angularLimitsMin=<0 0 0>, angularTarget=<1 0 0>, linearAxesMax=<2.5 2 2.3>, linearAxesMin=<-2.3 -2 -2.5>}
      localJointOffset=<0 0 0>
     }]
    relativeSpaceBone {boneName=pack}
   }
   n14 = Anim Dynamics - Bone: dyn_icicle_6(ComponentPose=@n13, Alpha=0.2, ExternalForce="0,0,0") {
    angularDampingOverride=1
    angularSpringConstant=20000
    bAngularSpring=true
    bGravityOverrideInSimSpace=true
    bLinearSpring=true
    bOverrideAngularDamping=true
    bOverrideLinearDamping=true
    bUseGravityOverride=true
    bUsePlanarLimit=false
    gravityScale=0.35
    linearDampingOverride=1
    linearSpringConstant=150
    simulationSpace=RootRelative
    alphaBoolBlend {blendOption=Linear}
    boundBone {boneName=dyn_icicle_6}
    componentAppliedLinearAccClamp=<10000 10000 10000>
    componentLinearAccScale=<0.12 0.12 0.12>
    componentLinearVelScale=<0.15 0.15 0.15>
    gravityOverride {z=-50}
    physicsBodyDefinitions [{
      sphereCollisionRadius=0
      boundBone {boneName=dyn_icicle_6}
      boxExtents=<11 19 19>
      constraintSetup {angularLimitsMax=<0 0 0>, angularLimitsMin=<0 0 0>, angularTarget=<1 0 0>, linearAxesMax=<2.2 2 2.5>, linearAxesMin=<-2.5 -2 -2.2>}
      localJointOffset=<0 0 0>
     }]
    relativeSpaceBone {boneName=pack}
   }
   Output Pose(Result=Component To Local(ComponentPose=@n14)) {layerGroup=None}
  ]}
 EventGraph {nodes [
   Event BlueprintUpdateAnimation()
   n1 = GetGameTimeInSeconds()
   n2 = GetOwnerSkeletalMeshComponent()
   > n3 = Random Vector Function!(
    execute
    CharacterPartAnimInstance:Custom Character Part Anim Instance Object Reference
    SkelMeshComponent:Mesh Component Object Reference=@n2
    JointName:Name=None
    GameTimeInSeconds:double=@n1
    TimeOffset:double=0
    PerlinScale:double=0
    Frequency:Vector="0.9,1.1,0.75"
    Amplitude:Vector="0.6,0.7,0.8"
    Offset:Vector="0,0,0"
   ) -> (then, VectorOut:Vector)
   > Set Shard 1 Movement(Shard 1 Movement:Vector=@n3) -> (Output_Get:Vector)
   > n4 = Random Vector Function(
    SkelMeshComponent=@n2
    JointName=None
    GameTimeInSeconds=@n1
    TimeOffset=0
    PerlinScale=0
    Frequency="1.2,0.8,0.6"
    Amplitude="0.9,0.1,1.1"
    Offset="0,0,0"
   )
   > Set Shard 2 Movement(Shard 2 Movement:Vector=@n4) -> (Output_Get:Vector)
   > n5 = Random Vector Function(
    SkelMeshComponent=@n2
    JointName=None
    GameTimeInSeconds=@n1
    TimeOffset=0
    PerlinScale=0
    Frequency="1,1,0.7"
    Amplitude="0.7,0.3,0.9"
    Offset="0,0,0"
   )
   > Set Shard 3 Movement(Shard 3 Movement:Vector=@n5) -> (Output_Get:Vector)
   > n6 = Random Vector Function(
    SkelMeshComponent=@n2
    JointName=None
    GameTimeInSeconds=@n1
    TimeOffset=0
    PerlinScale=0
    Frequency="1.1,0.75,0.9"
    Amplitude="1,0.5,0.6"
    Offset="0,0,0"
   )
   > Set Shard 4 Movement(Shard 4 Movement:Vector=@n6) -> (Output_Get:Vector)
   > n7 = Random Vector Function(
    SkelMeshComponent=@n2
    JointName=None
    GameTimeInSeconds=@n1
    TimeOffset=0
    PerlinScale=0
    Frequency="0.6,1.2,0.8"
    Amplitude="0.4,0.8,0.9"
    Offset="0,0,0"
   )
   > Set Shard 5 Movement(Shard 5 Movement:Vector=@n7) -> (Output_Get:Vector)
   > n8 = Random Vector Function(
    SkelMeshComponent=@n2
    JointName=None
    GameTimeInSeconds=@n1
    TimeOffset=0
    PerlinScale=0
    Frequency="0.5,0.7,1"
    Amplitude="0.9,0.7,0.7"
    Offset="0,0,0"
   )
   > Set Shard 6 Movement(Shard 6 Movement:Vector=@n8) -> (Output_Get:Vector)
   TryGetPawnOwner()
  ]}
]
variables [
 Shard 1 Movement {type=Vector}
 Shard 2 Movement {type=Vector}
 Shard 3 Movement {type=Vector}
 Shard 4 Movement {type=Vector}
 Shard 5 Movement {type=Vector}
 Shard 6 Movement {type=Vector}
]
